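# Niagara system for the Uproar emote that flashes a brief pale-blue mesh effect for a sound-barrier-breaking punch.
# Its single Punch emitter bursts very short-lived particles at a local offset, drawn by two mesh renderers with camera-maintained scale.
# Curves widen the mesh in Y and Z and fade alpha in and out, while a lerped position override slides it along X.
# Dynamic material parameters feed the shader an age-driven curve and a random value.
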NiagaraSystem /BRCosmetics/Effects/Fort_Effects/Emotes/Uproar/NS_uproar_Emote_PunchBreakSoundBarrior
dependencies {
 usedDataInterfaces [&NiagaraDataInterfaceCurve, &NiagaraDataInterfaceVectorCurve]
 usedDynamicInputs []
 usedModules [&SystemState]
}
emitters {Punch {
  emitterUpdateScript [
   &EmitterState(Life Cycle Mode=::NewEnumerator1, Loop Behavior=::NewEnumerator1, Loop Duration=1, UseLoopDelay=false)
   &SpawnBurst_Instantaneous(Use Spawn Probability=false, UseLoopCountLimit=false)
  ]
  particleSpawnScript [
   &InitializeParticle(
    Color=<rgba 0.3956 0.5145 0.7913 0.002>
    Lifetime Max=0.05
    Lifetime Min=0.1
    Lifetime Mode=::NewEnumerator1
    Mesh Scale=<0.08 0.08 0.08>
    Mesh Scale Mode=::NewEnumerator4
    Position Offset=<10 -3 1>
    Position Offset Coordinate Space=::Local
    UsePositionOffset=true
    Write Material Random=true
   )
   &InitialMeshOrientation(Mesh Axis To Orient=<0 -1 1>, Mesh Orientation Mode=::NewEnumerator2, Orientation Vector=<-1 0 0>, Use Rotation Vector=true) {enabled=false}
   /Niagara/Modules/Update/Camera/MaintainInCameraParticleScale()
  ]
  particleUpdateScript [
   &ParticleState()
   &ScaleColor(
    Scale Alpha=&FloatFromCurve(FloatCurve=DataInterface() {propertyValues {Curve {keys [
         {arriveTangent=5.8776, interpMode=RCIM_Cubic, leaveTangent=5.8776, tangentMode=RCTM_User, time=0, value=0}
         <at 0.1 2>
         <at 0.3 1>
         {
          arriveTangentWeight=0.4505
          interpMode=RCIM_Cubic
          leaveTangentWeight=0.4505
          tangentMode=RCTM_User
          tangentWeightMode=RCTWM_WeightedBoth
          time=0.5
          value=1
         }
         {arriveTangent=-6.2936, interpMode=RCIM_Cubic, leaveTangent=-6.2936, tangentMode=RCTM_User, time=1, value=0}
        ]}}})
    ScaleA=true
    ScaleRGB=false
   )
   &ScaleMeshSize(Scale Factor=&VectorFromCurve(
     VectorCurve=DataInterface() {propertyValues {
       XCurve {keys [<at 0 1 RCIM_Cubic>,<at 1 1 RCIM_Cubic>]}
       YCurve {keys [<at 0 0.75>,<at 0.75 1.45>,<at 1 3>]}
       ZCurve {keys [<at 0 0.75>,<at 0.75 1.45>,<at 1 3>]}
      }}
     Scale Curve=&VectorFromFloat(Value=1)
    ))
   &SolveForcesAndVelocity(Clamp Velocity=false, Limit Acceleration=false)
   /SetVariables(Particles.Position=&Add_Vector(
     A=Linked() {linkedVariable {name=Particles.Initial.Position, type=NiagaraPosition}}
     B=&Lerp_Vector(A=<-2 0 0>, B=<3 0 0>, Alpha=$Particles.NormalizedAge)
    )) {isSetParameters=true}
   &DynamicMaterialParameters(
    Index 0 Param 1=1
    Index 0 Param 2=1
    Index 0 Param 3=1
    Index 0 Param 4=&FloatFromCurve(FloatCurve=DataInterface() {propertyValues {Curve {keys [<at 0 5>,<at 1 1>]}}})
    Index 2 Param 1=1
    Index 2 Param 2=3
    Index 2 Param 3=0.25
    Index 3 Param 1=1
    Index 3 Param 2=&RandomRangeFloat(Minimum=0, Maximum=1)
    Index 3 Param 3=1
    Index 3 Param 4=1
    Module.Param0WriteEnabled=true
    Module.Param0WriteEnabled1=true
    Module.Param0WriteEnabled2=true
    Module.Param0WriteEnabled3=true
    Module.Param1WriteEnabled=true
    Module.Param1WriteEnabled1=true
    Module.Param1WriteEnabled2=true
    Module.Param1WriteEnabled3=true
    Module.Param2WriteEnabled=true
    Module.Param2WriteEnabled1=true
    Module.Param2WriteEnabled2=true
    Module.Param2WriteEnabled3=true
    Module.Param3WriteEnabled=true
    Module.Param3WriteEnabled1=true
    Module.Param3WriteEnabled2=true
    Module.Param3WriteEnabled3=true
    Write Parameter Index 1=true
    Write Parameter Index 2=true
    Write Parameter Index 3=true
   )
  ]
  rendererClasses [&NiagaraMeshRendererProperties]
  renderers [
   {rendererClass=&NiagaraMeshRendererProperties, rendererIndex=0}
   {rendererClass=&NiagaraMeshRendererProperties, rendererIndex=1}
  ]
 }}
summary {
 systemName=NS_uproar_Emote_PunchBreakSoundBarrior
 emitters [{emitterName=Punch, rendererClasses [&NiagaraMeshRendererProperties]}]
 userVariables []
}
