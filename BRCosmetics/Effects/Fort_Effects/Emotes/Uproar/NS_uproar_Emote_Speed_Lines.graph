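# Niagara system for the Uproar emote that draws short-lived, faint blue-white speed-line sprites.
# One sprite emitter spawns continuously plus an instantaneous burst on a torus, inheriting part of the source velocity, then applies point attraction.
# Curves fade alpha in and out, weighted by time of day, and swell then shrink sprite size over life.
# Speed-based sizing is disabled, and the exposed User.Speed parameter has no visible use.
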NiagaraSystem /BRCosmetics/Effects/Fort_Effects/Emotes/Uproar/NS_uproar_Emote_Speed_Lines
dependencies {
 usedDataInterfaces [&NiagaraDataInterfaceCurve, &NiagaraDataInterfaceVector2DCurve]
 usedDynamicInputs []
 usedModules [&SystemState]
}
emitters {FN_Sprite_SpeedLines {
  emitterUpdateScript [
   &EmitterState()
   &SpawnBurst_Instantaneous(Use Spawn Probability=false, UseLoopCountLimit=false)
   &SpawnRate(SpawnRate=90, Use Spawn Probability=false)
  ]
  particleSpawnScript [
   &InitializeParticle(
    Color=<rgba 0.3956 0.5145 0.7913 0.15>
    Lifetime Max=0.2
    Lifetime Min=0.25
    Lifetime Mode=::NewEnumerator1
    Position Offset Coordinate Space=::Local
    Sprite Size=<5 20>
    Sprite Size Mode=::NewEnumerator4
    UsePositionOffset=true
    Write Material Random=true
   ) {module=InitializeParticle001}
   &TorusLocation(Align Torus With Axis=true, Handle Radius=2.5, Large Radius=10, MaskSpawnGroup=false, Module.Override Randomness=false) {module=TorusLocation001}
   &InheritVelocity(
    Inherited Velocity Amount Scale=&VectorFromFloat(Value=0.6)
    Speed Limit=false
    UseSpeedThreshold=false
   )
  ]
  particleUpdateScript [
   &ParticleState()
   &ScaleColor(
    Scale Alpha=&FloatFromCurve(
     FloatCurve=DataInterface() {propertyValues {Curve {keys [
         {interpMode=RCIM_Cubic, tangentMode=RCTM_User, time=0, value=0}
         {interpMode=RCIM_Cubic, tangentMode=RCTM_User, time=0.5, value=1}
         {tangentMode=RCTM_User, time=1, value=0}
        ]}}}
     Scale Curve=&FNTimeOfDayScalar(Night Scale=2, Morning Scale=1, Evening Scale=1)
    )
    ScaleA=true
    ScaleRGB=true
   ) {module=ScaleColor001}
   &PointAttractionForce(Attraction Radius=10000000000, Attraction Strength=300, Kill Within Radius=false, Module.Use Falloff=true, UseAttractorOffset=true)
   &SolveForcesAndVelocity(Clamp Velocity=false, Limit Acceleration=false)
   &ScaleSpriteSizeBySpeed(Max Scale Factor=<0.5 1>, Min Scale Factor=<0 0>, Velocity Threshold=4250) {enabled=false}
   &ScaleSpriteSize(Scale Factor=&Vector2DFromCurve(
     Vector2Curve=DataInterface() {propertyValues {
       XCurve {keys [<at -0.0069 0.5 RCIM_Cubic>,<at 0.4918 1 RCIM_Cubic>,<at 1 0.5 RCIM_Cubic>]}
       YCurve {keys [<at 0 0.0245 RCIM_Cubic>, {arriveTangent=-0.0245, interpMode=RCIM_Cubic, leaveTangent=-0.0245, time=0.4975, value=1}, <at 1 0 RCIM_Cubic>]}
      }}
     ScaleCurve=<1 1>
    ))
  ]
  rendererClasses [&NiagaraSpriteRendererProperties]
  renderers [{rendererClass=&NiagaraSpriteRendererProperties}]
 }}
summary {
 systemName=NS_uproar_Emote_Speed_Lines
 emitters [{emitterName=FN_Sprite_SpeedLines, rendererClasses [&NiagaraSpriteRendererProperties]}]
 userVariables [User.Speed {type=NiagaraFloat, defaultValue=1, description=""}]
}
