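# Niagara effect that draws rising pale-blue speed-line sprites, named for the Uproar emote.
# One sprite emitter spawns short-lived elongated particles in a flattened ellipsoid, offset downward in local space, with mostly upward random velocity.
# Curves fade and stretch each streak in and out over its life, with alpha scaled by time of day to be stronger at night.
# A User.Speed parameter is exposed, but no shown module reads it.
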NiagaraSystem /BRCosmetics/Effects/Fort_Effects/Emotes/Uproar/NS_uproar_Emote_Speed_Lines2
dependencies {
 usedDataInterfaces [&NiagaraDataInterfaceCurve, &NiagaraDataInterfaceVector2DCurve]
 usedDynamicInputs []
 usedModules [&SystemState]
}
emitters {FN_Sprite_SpeedLines001 {
  emitterUpdateScript [
   &EmitterState()
   &SpawnBurst_Instantaneous(Use Spawn Probability=false, UseLoopCountLimit=false)
   &SpawnRate(SpawnRate=60, Use Spawn Probability=false)
  ]
  particleSpawnScript [
   &InitializeParticle(
    Color=<rgba 0.3956 0.5145 0.7913 0.25>
    Lifetime Max=0.2
    Lifetime Min=0.25
    Lifetime Mode=::NewEnumerator1
    Position Offset=<0 0 -30>
    Position Offset Coordinate Space=::Local
    Sprite Size=<5 17>
    Sprite Size Mode=::NewEnumerator4
    UsePositionOffset=true
    Write Material Random=true
   ) {module=InitializeParticle001}
   &SphereLocation(
    MaskSpawnGroup=false
    Non Uniform Scale=<3 1.5 0.25>
    Override Seed=false
    Sphere Radius=10
    Surface Only=false
    UseNonUniformScale=true
    UseOffset=false
    UseSphereOrientation=false
   )
   &AddVelocity(Velocity=&RandomRangeVector(Minimum=<30 30 180>, Maximum=<-30 -30 220>))
  ]
  particleUpdateScript [
   &ParticleState()
   &ScaleColor(
    Scale Alpha=&FloatFromCurve(
     FloatCurve=DataInterface() {propertyValues {Curve {keys [
         {interpMode=RCIM_Cubic, tangentMode=RCTM_User, time=0, value=0}
         {interpMode=RCIM_Cubic, tangentMode=RCTM_User, time=0.5, value=1}
         {tangentMode=RCTM_User, time=1, value=0}
        ]}}}
     Scale Curve=&FNTimeOfDayScalar(Night Scale=2, Morning Scale=1, Evening Scale=1)
    )
    ScaleA=true
    ScaleRGB=true
   ) {module=ScaleColor001}
   &SolveForcesAndVelocity(Clamp Velocity=false, Limit Acceleration=false)
   &ScaleSpriteSize(Scale Factor=&Vector2DFromCurve(
     Vector2Curve=DataInterface() {propertyValues {
       XCurve {keys [
         {arriveTangent=2.3707, interpMode=RCIM_Cubic, time=-0.0069, value=0.5}
         <at 0.5 1>
         {arriveTangent=-2.2178, interpMode=RCIM_Cubic, leaveTangent=-2.2178, tangentMode=RCTM_User, time=1, value=0.5}
        ]}
       YCurve {keys [
         {arriveTangent=2.423, interpMode=RCIM_Cubic, leaveTangent=2.423, leaveTangentWeight=0.872, tangentMode=RCTM_User, time=0, value=0.0245}
         {arriveTangent=-0.0245, interpMode=RCIM_Cubic, leaveTangent=-0.0245, time=0.4975, value=1}
         {arriveTangent=-2.7691, arriveTangentWeight=0.8531, interpMode=RCIM_Cubic, leaveTangent=-2.7691, tangentMode=RCTM_User, time=1, value=0}
        ]}
      }}
     ScaleCurve=<1 1>
    ))
  ]
  rendererClasses [&NiagaraSpriteRendererProperties]
  renderers [{rendererClass=&NiagaraSpriteRendererProperties}]
 }}
summary {
 systemName=NS_uproar_Emote_Speed_Lines2
 emitters [{emitterName=FN_Sprite_SpeedLines001, rendererClasses [&NiagaraSpriteRendererProperties]}]
 userVariables [User.Speed {type=NiagaraFloat, defaultValue=1, description=""}]
}
